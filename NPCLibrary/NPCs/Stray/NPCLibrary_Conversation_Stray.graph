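# Conversation database for the Stray NPC, a Fox Clan character, defining the dialogue that opens when a player talks to them.
# The entry tag leads to one speech node with a default line about faults splitting the land, plus prioritized contextual lines for a first-ever conversation or specific player outfits.
# That speech node branches through link nodes to shared NPC quests, basic services, and a Season 24 narrative quest conversation.
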ConversationDatabase /NPCLibrary/NPCs/Stray/NPCLibrary_Conversation_Stray
entry_points [{tag=Conversation.Entry.Stray, destinations [#1]}]
nodes [
 ConversationEntryPointNode_7 {
  id=#1
  class=ConversationEntryPointNode
  outputs [#2]
  properties {entryTag {tagName=Conversation.Entry.Stray}}
 }
 Speech_C_4 {
  id=#2
  class=Speech_C
  outputs [#3, #4, #5]
  properties {
   nodeName=""
   generalConfig {
    defaultMessage="I've been seeing faults develop, splitting the land. Something's happening."
    contextualMessages [
     {
      message=Talk fast. Fox Clan don't like being out in the open.
      contextRequirements [{requirement {class=FortControllerRequirement_IsFirstEverConversationWithNPC}}]
      priority {curve {rowName=Tandem.Speech.Priority.FirstEverInteraction}}
     }
     {
      message=Don't stray far from me. I don't want you getting hurt.
      contextRequirements [{requirement {
         class=FortControllerRequirement_HasCID
         properties {allowedCIds [/BRCosmetics/Athena/Items/Cosmetics/Characters/CID_161_Athena_Commando_M_Drift]}
        }}]
      priority {curve {rowName=Tandem.Speech.Priority.OutfitReaction}}
     }
     {
      message="They can hunt us, but they'll never find us."
      contextRequirements [{requirement {
         class=FortControllerRequirement_HasCID
         properties {allowedCIds [/BRCosmetics/Athena/Items/Cosmetics/Characters/CID_A_009_Athena_Commando_F_FoxWarrior_21B9R]}
        }}]
      priority {curve {rowName=Tandem.Speech.Priority.OutfitReaction}}
     }
     {
      message=I left home for a reason. And I'm not coming back.
      contextRequirements [{requirement {class=FortControllerRequirement_HasCID, properties {allowedCIds [/BRCosmetics/Athena/Items/Cosmetics/Characters/Character_NitroFlow]}}}]
      priority {curve {rowName=Tandem.Speech.Priority.OutfitReaction}}
     }
     {
      message=Fox Clan has no grudge against your Luminaries... but we trust no-one.
      contextRequirements [{requirement {class=FortControllerRequirement_HasCID, properties {allowedCIds [/BRCosmetics/Athena/Items/Cosmetics/Characters/Character_Nebula]}}}]
      priority {curve {rowName=Tandem.Speech.Priority.OutfitReaction}}
     }
     {
      message=Sometimes justice and peace mean different things.
      contextRequirements [{requirement {
         class=FortControllerRequirement_HasCID
         properties {allowedCIds [&CID_A_425_Athena_Commando_F_BlueJay]}
        }}]
      priority {curve {rowName=Tandem.Speech.Priority.OutfitReaction}}
     }
    ]
   }
  }
 }
 ConversationLinkNode_2 {
  id=#5
  class=ConversationLinkNode
  properties {remoteEntryTag {tagName=Conversation.Entry.Season24.Narrative.P02.Q06.Stray}}
 }
 ConversationLinkNode_0 {id=#4, class=ConversationLinkNode, properties {remoteEntryTag {tagName=Conversation.Entry.BasicServices}}}
 ConversationLinkNode_1 {id=#3, class=ConversationLinkNode, properties {remoteEntryTag {tagName=Conversation.Entry.NPCLibrary.Quests}}}
]
